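# Unlit alpha-composite material in the BiomeColorGenerator tool, named for drawing landscape biome color without alpha.
# It samples the RT_LandscapeBiomeCapture render target and feeds its RGB into two separate branches: a divide by noise and a multiply with no second input wired.
# Neither branch connects to a material output, so the graph as it stands looks unfinished or left over from experimentation.
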Material /FigmentEnvironment/Tools/BiomeColorGenerator/M_DrawLandscapeBiomeNoAlpha
graph {
 properties {bAutomaticallySetUsageInEditor=false, blendMode=BLEND_AlphaComposite, shadingModel=MSM_Unlit, usedShadingModels=Unlit}
 nodes [
  n1 = TextureSample() {texture=&/RT_LandscapeBiomeCapture}
  Divide(A=@n1.RGB, B=Noise!(World Position, FilterWidth))
  Multiply(A=@n1.RGB)
 ]
}
