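# Editor utility actor that bakes a top-down landscape biome color map into a texture asset, positioning itself high above the origin.
# Setup switches the landscape material instance to RenderTintUnlit and toggles ocean water visibility; SetBack clears the switch and toggles water again.
# CaptureScene renders the scene capture; SaveTexture draws a no-alpha material into a render target, converts it to the named texture, sets sRGB and compression, and saves.
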EditorUtilityBlueprint /FigmentEnvironment/Tools/BiomeColorGenerator/EUA_LandscapeBiomeColorGenerator
graphs [
 01_SetupMaterials {nodes [
   n1 = GetEditorWorld(execute=@n14.then, self=Get UnrealEditorSubsystem())
   n2 = StripRightMostFolderFromPath_Util!(
    execute=@n20.then
    Path:String=Get Path() -> (Path:String)
    WorldContextObject:Object Reference=@n1
    Strip Count:int=1
    __WorldContext:Object Reference=@n1
   ) -> (then, FinalPath1:String, AssetName:String)
   > n3 = StripRightMostFolderFromPath_Util(Path=@n2, WorldContextObject=@n1, Strip Count=1, __WorldContext=@n1)
   n4 = Append(A=@n3, B=/Textures/)
   n5 = GetOrCreateTexture2D_Util!(
    execute
    PackagePath:String=@n4
    Asset Name:String=Get TextureName() -> (TextureName:String)
    WorldContextObject:Object Reference=@n1
    __WorldContext:Object Reference=@n1
   ) -> (then, Texture:Texture 2D Object Reference)
   > CaptureScene(self:Scene Capture Component 2D Object Reference=Get SceneCaptureComponent2D() -> (SceneCaptureComponent2D:Scene Capture Component 2D Object Reference))
   > ConvertRenderTargetToTexture2DEditorOnly(
    WorldContextObject=@n1
    RenderTarget=Get TextureTarget!(self:Scene Capture Component 2D Object Reference=Get SceneCaptureComponent2D()) -> (TextureTarget:Texture Render Target 2D Object Reference)
    Texture=@n5
   )
   > n6 = GetAllActorsOfClass!(
    execute
    WorldContextObject:Object Reference=@n1
    ActorClass:Actor Class Reference=/Game/Athena/Apollo/Environments/Landscape/Water/FortWaterBodyBP.FortWaterBodyBP_C
   ) -> (then, OutActors:Array of Fort Water Body BP Object References)
   > n7 = For Each Loop(Array:Array of Fort Water Body BP Object References=@n6) -> (Array Element:Fort Water Body BP Object Reference)
   n8 = Get WaterBodyComponent!(self:Water Body Object Reference=@n7) -> (WaterBodyComponent:Water Body Component Object Reference)
   > ToggleVisibility!(execute, self:Scene Component Object Reference=@n8, bPropagateToChildren:bool=false) -> (then)
   > ReloadPackages!(
    execute
    PackagesToReload:Array of Package Object References=Make Array("[0]":Package Object Reference=GetPackageForObject!(Object:Object Reference=@n8) -> (ReturnValue:Package Object Reference)) -> (Array:Array of Package Object References)
    InteractionMode:EReloadPackagesInteractionMode Enum=Interactive
   ) -> (then, bOutAnyPackagesReloaded:bool, OutErrorMessage:Text)
   n9 = Get As Material Instance Constant() -> (As Material Instance Constant:Material Instance Constant Object Reference)
   > SetMaterialInstanceStaticSwitchParameterValue!(
    execute
    Instance:Material Instance Constant Object Reference=@n9
    ParameterName:Name=RenderTintUnlit
    Value:bool=false
    Association:EMaterialParameterAssociation Enum=GlobalParameter
    bUpdateMaterialInstance:bool=true
   ) -> (then, ReturnValue:bool)
   > UpdateMaterialInstance!(execute, Instance:Material Instance Constant Object Reference=@n9) -> (then)
   n10 = GetAllActorsOfClass(execute=@n1.then, WorldContextObject=@n1, ActorClass=/Script/Landscape) -> (OutActors:Array of Landscape Object References)
   n11 = Get LandscapeMaterial!(self:Landscape Proxy Object Reference="Get (a copy)"(Array:Array of Landscape Object References=@n10) -> (Output:Landscape Object Reference)) -> (LandscapeMaterial:Material Interface Object Reference)
   n12 = GetStaticSwitchParameterNames!(execute, Material:Material Interface Object Reference=@n11) -> (then, ParameterNames:Array of Names)
   > n13 = For Each Loop(Array:Array of Names=@n12) -> (Array Element:Name)
   > PrintString(WorldContextObject=@n1, InString="To String (Name)"(InName=@n13))
   n14 = 01_SetupMaterials() {is=title K2Node_FunctionEntry}
   PrintString(WorldContextObject=@n1, InString=GetDisplayName())
   n15 = Cast To MaterialInstance!(execute=@n10.then, Object=@n11) -> (then, CastFailed, AsMaterial Instance:Material Instance Object Reference) {is=title K2Node_DynamicCast}
   > n16 = Cast To MaterialInstanceConstant(Object=@n15)
   GetPackageForObject(Object=@n16)
   Get Object Path String(Object=@n16)
   n17 = Get MI() -> (MI:Material Instance Constant Object Reference)
   GetSystemPath!(Object:Object Reference=@n17) -> (ReturnValue:String)
   ReloadPackages(PackagesToReload=Make Array("[0]"=GetPackageForObject(Object=@n17)), InteractionMode=Interactive)
   PrintString(
    WorldContextObject=@n1
    InString=GetDisplayName(Object=Cast To MaterialInterface!(execute, Object=@n11) -> (then, CastFailed, AsMaterial Interface:Material Interface Object Reference) {is=title K2Node_DynamicCast})
   )
   Cast To Class!(Object:Object Reference=@n11, Class:Object Class Reference) -> (ReturnValue:Object Class Reference)
   Get SceneCaptureComponent2D()
   Set Path(execute=@n23.then, Path:String=@n4) -> (Output_Get:String)
   n18 = GetAllActorsOfClass(
    execute=@n22.then
    WorldContextObject=@n1
    ActorClass=/Game/Athena/Asteria/Environments/Water/Asteria_Waterbody_Ocean_Parent.Asteria_Waterbody_Ocean_Parent_C
   ) -> (OutActors:Array of Asteria Waterbody Ocean Parent Object References)
   > n19 = For Each Loop(Array:Array of Fort Water Body BP Object References=@n18) -> (Array Element:Fort Water Body BP Object Reference)
   > n20 = ToggleVisibility(self=Get WaterBodyComponent(self=@n19), bPropagateToChildren=false)
   n21 = Set As Material Instance Constant(execute=@n24.then, As Material Instance Constant:Material Instance Constant Object Reference=@n16) -> (Output_Get:Material Instance Constant Object Reference)
   > SetMaterialInstanceStaticSwitchParameterValue(Instance=@n21, ParameterName=RenderTintUnlit, Value=true, Association=GlobalParameter, bUpdateMaterialInstance=true)
   > n22 = UpdateMaterialInstance(Instance=@n21)
   n23 = PrintString(execute=@n3.then, WorldContextObject=@n1, InString=@n4)
   n24 = Branch(
    execute=@n16.then
    Condition="Equal (Object)"(A=Get Parent!(self:Material Instance Object Reference=@n15) -> (Parent:Material Interface Object Reference))
   )
  ]}
 CaptureScene {nodes [
   CaptureScene() {is=title K2Node_FunctionEntry}
   n1 = Get SceneCaptureComponent2D()
   > CaptureScene(self=@n1) {is=title K2Node_CallFunction}
   n2 = Get TextureTarget(self=@n1)
   > Set AdjustMaxAlpha!(execute, AdjustMaxAlpha:float=0, self:Texture Object Reference=@n2) -> (then, Output_Get:float)
   > Set AdjustMinAlpha!(execute, AdjustMinAlpha:float=0, self:Texture Object Reference=@n2) -> (then, Output_Get:float)
  ]}
 EventGraph {nodes [
   Event BeginPlay()
   Event ActorBeginOverlap()
   Event Tick!() -> (OutputDelegate, then, DeltaSeconds:float)
  ]}
 SaveTexture {nodes [
   SaveTexture() {is=title K2Node_FunctionEntry}
   > n1 = GetEditorWorld(self=Get UnrealEditorSubsystem())
   n2 = Get Path()
   n3 = Get TextureName()
   > n4 = GetOrCreateTexture2D_Util(PackagePath=@n2, Asset Name=@n3, WorldContextObject=@n1, __WorldContext=@n1)
   > ClearRenderTarget2D!(
    execute
    WorldContextObject:Object Reference=@n1
    TextureRenderTarget:Texture Render Target 2D Object Reference=&/RT_LandscapeBiomeCaptureNoAlpha
    ClearColor:Linear Color Structure=(R=0,G=0,B=0,A=0.5)
   ) -> (then)
   > DrawMaterialToRenderTarget!(
    execute
    WorldContextObject:Object Reference=@n1
    TextureRenderTarget:Texture Render Target 2D Object Reference=&RT_LandscapeBiomeCaptureNoAlpha
    Material:Material Interface Object Reference=&/M_DrawLandscapeBiomeNoAlpha
   ) -> (then)
   > ConvertRenderTargetToTexture2DEditorOnly(WorldContextObject=@n1, RenderTarget=&RT_LandscapeBiomeCaptureNoAlpha, Texture=@n4)
   n5 = Append!(A=@n2, B=/, C:String=@n3) -> (ReturnValue)
   > n6 = SaveAsset!(execute, AssetToSave:String=@n5, bOnlyIfIsDirty:bool=false) -> (then, ReturnValue:bool)
   > Branch(Condition=@n6)
   > Set CompressionSettings(CompressionSettings=TC_Default, self=@n4)
   > Set CompressionNoAlpha!(execute, CompressionNoAlpha:bool=true, self:Texture Object Reference=@n4) -> (then, Output_Get:bool)
   > Set SRGB(SRGB=true, self=@n4)
   > SaveAsset(AssetToSave=@n5, bOnlyIfIsDirty=false)
   Set AdjustMaxAlpha(AdjustMaxAlpha=0.5, self=@n4)
   > Set AdjustMinAlpha(AdjustMinAlpha=0.6, self=@n4)
   > InvertAlphaChannel!(execute, Texture:Texture 2D Object Reference=@n4) -> (then)
   Get TextureTarget(self=Get SceneCaptureComponent2D())
  ]}
 SetBack {nodes [
   n1 = GetEditorWorld(execute=@n10.then, self=Get UnrealEditorSubsystem())
   > n2 = GetAllActorsOfClass(WorldContextObject=@n1, ActorClass=&Landscape) -> (OutActors:Array of Landscape Object References)
   n3 = Get LandscapeMaterial(self="Get (a copy)"(Array=@n2))
   n4 = Cast To MaterialInstanceConstant(execute=[@n2.then, @n9.then], Object=@n3)
   n5 = Set As Material Instance Constant(execute=@n12.then, As Material Instance Constant=@n4)
   n6 = SetMaterialInstanceStaticSwitchParameterValue(Instance=@n5, ParameterName=RenderTintUnlit, Value=false, Association=GlobalParameter, bUpdateMaterialInstance=true)
   n7 = GetAllActorsOfClass(
    execute=[@n6.then, @n11.then]
    WorldContextObject=@n1
    ActorClass=&Asteria_Waterbody_Ocean_Parent_C
   ) -> (OutActors:Array of Asteria Waterbody Ocean Parent Object References)
   > n8 = For Each Loop(Array=@n7)
   > ToggleVisibility(self=Get WaterBodyComponent(self=@n8), bPropagateToChildren=false)
   n9 = Cast To MaterialInstance(Object=@n3)
   n10 = SetBack() {is=title K2Node_FunctionEntry}
   SetMaterialInstanceStaticSwitchParameterValue(
    Instance=Get As Material Instance Constant()
    ParameterName=RenderTintUnlit
    Value=false
    Association=GlobalParameter
    bUpdateMaterialInstance=true
   )
   SetMaterialInstanceStaticSwitchParameterValue(execute=@n5.then, Instance=@n5, ParameterName=RenderTintUnlit, Value=false, Association=GlobalParameter, bUpdateMaterialInstance=true)
   > n11 = UpdateMaterialInstance(Instance=@n5)
   n12 = Branch(execute=@n4.then, Condition="Equal (Object)"(A=Get Parent(self=@n4)))
  ]}
 UserConstructionScript {nodes [
   Construction Script()
   > Set Actor Transform!(
    execute
    self:Actor Object Reference
    NewTransform:"Transform (by ref)"=MakeTransform(Location="0,0,20000", Rotation="0, 0, 0")
    bSweep:bool=false
    bTeleport:bool=false
   ) -> (then, SweepHitResult:Hit Result Structure, ReturnValue:bool) {is=title K2Node_CallFunction}
  ]}
]
variables [
 As Material Instance Constant {type=MaterialInstanceConstant, default=null}
 NewVar {type=boolean}
 Path {type=string, default=/FigmentEnvironment/Blueprints/Cliffs/Biome/Textures/}
 TextureName {type=string, default=T_AthenaBiomeColorMap_S04}
]
